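# Sound cue for a trigger plate powering on, in the Battle Lab gameplay cue sounds and the Creative gadgets sound class.
# A single Creative ActionTrigger power-on wave plays through a Modulator, with the cue raising its pitch and lowering its volume.
# Inline attenuation overrides take the place of the referenced small-toy attenuation asset, and a concurrency override caps simultaneous instances.
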SoundCue /FNGameplayCues/Sounds/BattleLab/TriggerPlate/TriggerPlate_PowerOn_Cue
root=Modulator_0
nodes [
 {id=Modulator_0, children [WavePlayer_0]}
 {id=WavePlayer_0, properties {soundWaveAssetPtr=/Game/Sounds/Creative/Toys/ActionTrigger/ActionTrigger_PowerOn_01}}
]
properties {
 attenuationSettings=/Game/Sounds/Creative/Attenuations/Fort_Creative_SmallToy_Att
 bOverrideAttenuation=true
 bOverrideConcurrency=true
 maxDistance=768
 pitchMultiplier=1.4
 soundClassObject=&Fort_Gadgets_Creative
 volumeMultiplier=0.2
 attenuationOverrides {falloffDistance=512, attenuationShapeExtents {x=256}}
 concurrencyOverrides {maxCount=2}
}
